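# Material function that builds world position offset and custom UV data for animated floating score numbers.
# The floored UV1 X index selects per-card vector parameters for a background and five glyph cards, supplying offset, size, digit-strip UV shift, start time and random value.
# Time over lifespan samples curve textures that animate scale and feed custom UV outputs.
# Cards become camera-facing sprites, spaced along view X and pushed by a camera offset.
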MaterialFunction /Game/UI/DamageText/Score/ScoreNumberWPO
graph {nodes [
  n1 = VectorParameter() {desc=0, group=+-, parameterName=1c, defaultValue=<rgba 0 0.8965 0 0>}
  n2 = Constant()
  n3 = VectorParameter() {group=+-, parameterName=0c, defaultValue=<rgba 1 0.5575 0 0>}
  n4 = AppendVector(A=@n3.RGB, B=@n3.RGB)
  n5 = AppendVector(A=@n1.RGB, B=@n1.RGB)
  n6 = TextureCoordinate() {coordinateIndex=1}
  n7 = Floor(None=ComponentMask(None=FunctionInput(Preview=@n6) {bUsePreviewValueAsDefault=true, inputName=Debug texture coordinate input, inputType=FunctionInput_Vector2}) {r=true})
  n8 = Clamp(None=@n7)
  n9 = VectorParameter() {desc=0, group=+-, parameterName=2c, defaultValue=<rgba 0 0.2989 0 0>}
  n10 = Subtract(A=@n7)
  n11 = Clamp(None=@n10)
  n12 = VectorParameter() {desc=0, group=+-, parameterName=3c, defaultValue=<rgba 0 0.179 0 0>}
  n13 = Subtract(A=@n10)
  n14 = Clamp(None=@n13)
  n15 = LinearInterpolate(
   A=LinearInterpolate(A=LinearInterpolate(A=@n4, B=@n5, Alpha=@n8), B=AppendVector(A=@n9.RGB, B=@n9.RGB), Alpha=@n11)
   B=AppendVector(A=@n12.RGB, B=@n12.RGB)
   Alpha=@n14
  )
  n16 = VectorParameter() {desc=0, group=+-, parameterName=4c, defaultValue=<rgba 0 0.3333 0 0>}
  n17 = AppendVector(A=@n16.RGB, B=@n16.RGB)
  n18 = Clamp(None=Subtract(A=@n13))
  n19 = Clamp(None=Add(A=@n7))
  n20 = MaterialFunctionCall(Float4=LinearInterpolate(
    A=MaterialFunctionCall(X=@n1.R, Y=@n2, Z=@n2, A=@n2) {materialFunction=&MakeFloat4}.Result
    B=LinearInterpolate(A=@n15, B=@n17, Alpha=@n18)
    Alpha=@n19
   )) {materialFunction=&BreakOutFloat4Components}
  n21 = Divide(
   A=Subtract(A=@n20.R, B=Time() {bIgnorePause=true})
   B=FunctionInput(Preview=ScalarParameter() {defaultValue=1, parameterName=Animation Lifespan}) {bUsePreviewValueAsDefault=true, inputName=Animation Lifespan Override, inputType=FunctionInput_Scalar, sortPriority=15}
  )
  n22 = FunctionInput(Preview=OneMinus(None=Clamp(None=@n21))) {bUsePreviewValueAsDefault=true, inputName=Animation, inputType=FunctionInput_Scalar, sortPriority=2}
  n23 = TextureSample(UVs=@n22) {desc="X Scale, Y Scale,  ", samplerType=SAMPLERTYPE_LinearColor, texture=&/ScoreAnimation_RGXYScale_BRdynamic_AAlpha}
  FunctionOutput(None=AppendVector(A=@n23.B, B=@n23.B)) {outputName=To custom uvs 3 - B Randomization A Alpha}
  FunctionOutput(None=@n22) {outputName=n 0-1 animation}
  n24 = MaterialFunctionCall(Float2=TextureCoordinate() {vTiling=0.1}) {materialFunction=&BreakOutFloat2Components}
  n25 = VectorParameter() {parameterName=Background_b, defaultValue=<rgba 150 150 0 0>}
  n26 = VectorParameter() {group="Scale, rotation angle and values", parameterName=0b, defaultValue=<rgba 30 50 0 0>}
  n27 = VectorParameter() {group="Scale, rotation angle and values", parameterName=1b, defaultValue=<rgba 30 50 0 1>}
  n28 = VectorParameter() {group="Scale, rotation angle and values", parameterName=2b, defaultValue=<rgba 30 50 0 1>}
  n29 = LinearInterpolate(
   A=LinearInterpolate(A=AppendVector(A=@n26.RGB, B=@n26.RGB), B=AppendVector(A=@n27.RGB, B=@n27.RGB), Alpha=@n8)
   B=AppendVector(A=@n28.RGB, B=@n28.RGB)
   Alpha=@n11
  )
  n30 = VectorParameter() {group="Scale, rotation angle and values", parameterName=3b, defaultValue=<rgba 30 50 2 -1>}
  n31 = AppendVector(A=@n30.RGB, B=@n30.RGB)
  n32 = VectorParameter() {group="Scale, rotation angle and values", parameterName=4b, defaultValue=<rgba 30 50 1 1>}
  n33 = LinearInterpolate(
   A=AppendVector(A=@n25.RGB, B=@n25.RGB)
   B=LinearInterpolate(A=LinearInterpolate(A=@n29, B=@n31, Alpha=@n14), B=AppendVector(A=@n32.RGB, B=@n32.RGB), Alpha=@n18)
   Alpha=@n19
  )
  n34 = MaterialFunctionCall(Float4=@n33) {desc="rg = scale xy, b = rotation angle ", materialFunction=&BreakOutFloat4Components}
  n35 = Multiply(A=@n34.B) {constB=0.1}
  FunctionOutput(None=AppendVector(A=@n24.R, B=Add(A=@n24.G, B=@n35))) {outputName=CustomUV1 - number uvs, sortPriority=10}
  FunctionOutput(None=@n7) {outputName="optional copy of uv 1 (card selection)", sortPriority=10}
  FunctionOutput(None=@n20.G) {outputName="Random Number  - needs custom uvs", sortPriority=-1}
  n36 = Add(
   A=ScalarParameter() {defaultValue=472.4705, parameterName=Camera offset}
   B=FunctionInput() {bUsePreviewValueAsDefault=true, inputName="Additional Camera Offset ", inputType=FunctionInput_Scalar, sortPriority=10}
  )
  n37 = VectorParameter() {parameterName=Background_a, defaultValue=<rgba 1 0 1 1>}
  n38 = VectorParameter() {desc=alpha contains offsets, group=Placement and offsets, parameterName=0a, defaultValue=<rgba 0 0 0 0.5>}
  n39 = AppendVector(A=@n38.RGB, B=@n38.RGB)
  n40 = VectorParameter() {group=Placement and offsets, parameterName=1a, defaultValue=<rgba 0 0 0 1.2521>}
  n41 = AppendVector(A=@n40.RGB, B=@n40.RGB)
  n42 = VectorParameter() {group=Placement and offsets, parameterName=2a, defaultValue=<rgba 0 0 0 2.0847>}
  n43 = VectorParameter() {group=Placement and offsets, parameterName=3a, defaultValue=<rgba 0 0 0 2.9269>}
  n44 = LinearInterpolate(
   A=LinearInterpolate(A=LinearInterpolate(A=@n39, B=@n41, Alpha=@n8), B=AppendVector(A=@n42.RGB, B=@n42.RGB), Alpha=@n11)
   B=AppendVector(A=@n43.RGB, B=@n43.RGB)
   Alpha=@n14
  )
  n45 = VectorParameter() {group=Placement and offsets, parameterName=4a, defaultValue=<rgba 0 0 0 3.6974>}
  n46 = AppendVector(A=@n45.RGB, B=@n45.RGB)
  n47 = ComponentMask(None=LinearInterpolate(A=AppendVector(A=@n37.RGB, B=@n38.A), B=LinearInterpolate(A=@n44, B=@n46, Alpha=@n18), Alpha=@n19)) {a=true, desc=offset}
  n48 = FunctionInput() {bUsePreviewValueAsDefault=true, inputName="Number Offset Scale ", inputType=FunctionInput_Scalar, previewValue=<1 0 0 1>}
  n49 = FunctionInput() {bUsePreviewValueAsDefault=true, inputName=Number Scale, inputType=FunctionInput_Scalar, previewValue=<1 0 0 1>}
  n50 = ComponentMask(None=@n33) {g=true, r=true}
  n51 = Multiply(A=Multiply(A=Multiply(A=@n49, B=@n50), B=@n23.R) {desc=" XY dimensions"}, B=AppendVector(A=Constant() {r=1}, B=@n23.G))
  n52 = FunctionInput() {
   bUsePreviewValueAsDefault=true
   inputName="Additonal Screenspace XY offset "
   inputType=FunctionInput_Vector2
   sortPriority=11
   previewValue=<0 0 0 0>
  }
  n53 = MaterialFunctionCall(Float2=Add(A=@n51, B=@n52)) {materialFunction=&BreakOutFloat2Components}
  n54 = Multiply(
   A=Transform(None=Constant3Vector() {constant=<rgba 1 0 0 0>}) {transformSourceType=TRANSFORMSOURCE_View}
   B=Multiply(A=Multiply(A=@n47, B=@n48), B=@n53.R) {desc=scale x by offset amount - accum 1s}
  ) {constB=50}
  n55 = Transform(None=Multiply(A=MaterialFunctionCall(Y=@n53.G) {materialFunction=&MakeFloat3}.Result) {constB=0.5}) {transformSourceType=TRANSFORMSOURCE_View}
  n56 = FunctionInput() {bUsePreviewValueAsDefault=true, inputName=Scale Modulation, inputType=FunctionInput_Vector2, previewValue=<1 1 0 1>}
  n57 = Add(
   A=MaterialFunctionCall() {materialFunction=&ComponentPivotLocation}.Result
   B=MaterialFunctionCall!(Center Location=Add(A=@n54, B=@n55), 0-1 UVs=TextureCoordinate(), XY scale=Multiply(A=@n51, B=@n56), OverridePivot, "Normals (Optional)") {materialFunction=&Sprite}.World position offset
  )
  n58 = Add(A=@n57, B=WorldPosition().XYZ)
  FunctionOutput(None=Add(
    A=MaterialFunctionCall(Offset Amount=@n36, World Position=@n58) {materialFunction=/Engine/Functions/Engine_MaterialFunctions02/WorldPositionOffset/CameraOffset}.Clamped Camera Offset
    B=@n57
   )) {outputName=WPO}
  n59 = ComponentMask(None=TextureCoordinate() {coordinateIndex=1}) {r=true}
  n60 = Constant() {r=1}
  n61 = Constant()
  FunctionOutput(None=AppendVector(A=If(A=@n59, B=@n60, "A > B"=@n61, "A < B"=@n60), B=@n23.R)) {outputName="CustomUV2 - +- Card=1 others 0 _ And x scale", sortPriority=11}
  FunctionOutput(None=TextureSample(UVs=@n22) {samplerType=SAMPLERTYPE_LinearColor, texture=&/ScoreAnimation_RGB_DynamicBGA_A_Unused}.RGB) {outputName="To Custom UVS - R- Timeline for rippple - G - Float Uniform - B - White ", sortPriority=11}
 ]}
