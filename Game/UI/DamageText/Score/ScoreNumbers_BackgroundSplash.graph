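# Additive unlit material for the background splash behind damage score numbers, with world position offset logic for sprite placement, camera offset and score-number animation.
# A radial mask drives a ripple texture, scrolled by a timeline from custom UVs and multiplied with a tiled grid.
# Two color parameters tint the result, with stepped screen-aligned noise and randomized, gradient-based hue shift.
# Opacity combines custom-UV alpha, an edge falloff and depth fade.
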Material /Game/UI/DamageText/Score/ScoreNumbers_BackgroundSplash
graph {
 properties {
  bEnableResponsiveAA=true
  blendMode=BLEND_Additive
  numCustomizedUVs=8
  pixelDepthOffsetMode=PDOM_Legacy
  refractionMethod=RM_IndexOfRefraction
  shadingModel=MSM_Unlit
  usedShadingModels=Unlit
 }
 nodes [
  n1 = TextureSample!(UVs=TextureCoordinate() {uTiling=5, vTiling=5}, Tex, Bias=Constant() {r=1}, Apply View MipBias) {mipValueMode=TMVM_MipBias, samplerType=SAMPLERTYPE_LinearColor, texture=/Game/Gadgets/Assets/AirStrike/Textures/T_Airstrike_Grid_01}
  n2 = SphereMask(A=TextureCoordinate(), B=Constant() {r=0.5}) {attenuationRadius=0.5, hardnessPercent=0}
  n3 = OneMinus(None=@n2)
  n4 = MaterialFunctionCall(Float2=TextureCoordinate() {coordinateIndex=4}) {
   desc="R=Timeline for ripple G= Float Uniform"
   materialFunction=&BreakOutFloat2Components
  }
  n5 = Multiply(
   A=@n2
   B=TextureSample!(UVs=Add(A=@n3, B=@n4.R), Tex, Apply View MipBias) {samplerType=SAMPLERTYPE_LinearColor, texture=&/ScoreNumber_ripple}.RGB
  )
  n6 = Add(A=Multiply(A=@n1.R, B=@n5), B=Multiply(A=@n5) {constB=0.05})
  Add(A=@n6)
  n7 = ScalarParameter() {defaultValue=472.4705, parameterName=Camera offset}
  n8 = Multiply(
   A=Transform(None=Constant3Vector() {constant=<rgba 1 0 0 0>}) {transformSourceType=TRANSFORMSOURCE_View}
   B=ScalarParameter() {defaultValue=-25, parameterName=X Offset}
  ) {constB=50}
  n9 = Multiply(
   A=Transform(None=Constant3Vector() {constant=<rgba 0 1 0 0>}) {transformSourceType=TRANSFORMSOURCE_View}
   B=ScalarParameter() {defaultValue=25, parameterName=Y Offset}
  ) {constB=50}
  n10 = Add(
   A=MaterialFunctionCall() {materialFunction=&ComponentPivotLocation}.Result
   B=MaterialFunctionCall!(Center Location=Add(A=@n8, B=@n9), 0-1 UVs=TextureCoordinate(), XY scale=Constant() {r=80}, OverridePivot, "Normals (Optional)") {materialFunction=&Sprite}.World position offset
  )
  n11 = WorldPosition()
  Add(
   A=MaterialFunctionCall(Offset Amount=Add(A=@n7) {constB=-10}, World Position=Add(A=@n10, B=@n11.XYZ)) {materialFunction=/Engine/Functions/Engine_MaterialFunctions02/WorldPositionOffset/CameraOffset}.Clamped Camera Offset
   B=@n10
  )
  n12 = MaterialFunctionCall!(
   Scale Modulation=ComponentMask(None=VectorParameter() {group=Visuals, parameterName=XYScaleModulation, defaultValue=<rgba 10 10 10 1>}.RGB) {g=true, r=true}
   Number Scale=ScalarParameter() {defaultValue=3, parameterName=NumberScale}
   "Number Offset Scale "=ScalarParameter() {defaultValue=3, parameterName=NumberOffsetScale}
   Debug texture coordinate input
   Animation
   "Additional Camera Offset "
   "Additonal Screenspace XY offset "
   Animation Lifespan Override
  ) {materialFunction=&/ScoreNumberWPO}
  ComponentMask(None=@n12."To Custom UVS - R- Timeline for rippple - G - Float Uniform - B - White ") {g=true, r=true}
  ComponentMask(None=@n12."To Custom UVS - R- Timeline for rippple - G - Float Uniform - B - White ") {b=true}
  Constant() {r=1}
  n13 = MaterialFunctionCall(Float2=TextureCoordinate() {coordinateIndex=3, desc="x= B Randomization  y= A Alpha"}) {materialFunction=&BreakOutFloat2Components}
  n14 = TextureCoordinate()
  n15 = OneMinus(None=Power(Base=ConstantBiasScale(None=@n14) {bias=-0.5, scale=2}) {constExponent=8})
  DepthFade(
   Opacity=Multiply(A=@n13.G, B=MaterialFunctionCall!(f2=@n15, f3, f4) {materialFunction=/Game/Packages/Fortnite_UniqueMaterials_1/VectorComponentOps}.f2 Mul)
   FadeDistance=Constant() {r=64}
  )
  n16 = ScalarParameter() {defaultValue=-0.0408, parameterName=Color Gradient Shift On Top}
  n17 = ScalarParameter() {defaultValue=0.0183, parameterName=Color Gradient Shift On Bottom}
  n18 = MaterialFunctionCall(Float2=TextureCoordinate()) {materialFunction=&BreakOutFloat2Components}
  n19 = Multiply(
   A=ScalarParameter() {defaultValue=-0.0069, parameterName=Noise}
   B=MaterialFunctionCall!(Center Point World Position) {materialFunction=/Game/Packages/Fortnite_UniqueMaterials_1/Effects/WorldAlignedScreenCoordinates}.Result
  ) {constA=42811.2383}
  n20 = Add(
   A=MaterialFunctionCall!(TimeSpeed=Constant() {r=0.35}, Jump Amount Per Step, Number of Steps Per Unit of Time, Time Offset) {materialFunction=/Engine/Functions/Engine_MaterialFunctions02/SteppingPannerTime}.Result
   B=Multiply(A=@n19) {constB=0.5}
  )
  n21 = Add(
   A=@n19
   B=MaterialFunctionCall(TimeSpeed=Constant() {r=0.5}) {materialFunction=&SteppingPannerTime}.Result
  )
  n22 = LinearInterpolate(B=Add(
    A=TextureSample(UVs=@n20) {samplerType=SAMPLERTYPE_Normal, texture=/Game/Global/Textures/Default/Noise/64x64Noise}.RGB
    B=TextureSample(UVs=@n21) {samplerType=SAMPLERTYPE_Normal, texture=&64x64Noise}.RGB
   )) {constA=0.5, constAlpha=0.025, constB=0.5}
  n23 = ComponentMask(None=TextureCoordinate() {coordinateIndex=5}) {desc="R = Blue Dynamic White", r=true}
  n24 = VectorParameter() {parameterName=Color0, defaultValue=<rgba 0 0.0362 0.24 1>}
  n25 = VectorParameter() {parameterName=Color1, defaultValue=<rgba 0.3513 0.5407 0.865 0.01>}
  n26 = Multiply(A=@n6)
  n27 = Add(B=Multiply(A=Add(A=@n23, B=@n24.RGB), B=Multiply(A=@n25.RGB, B=@n26)))
  n28 = Multiply(A=@n4.G, B=@n27)
  MaterialFunctionCall(
   Hue Shift Percentage=Add(A=Multiply(A=@n13.R) {constB=0.15}, B=LinearInterpolate(A=@n16, B=@n17, Alpha=@n18.G) {constA=1, constB=0.5})
   Texture=Multiply(A=Add(A=@n22) {constB=0.5}, B=Add(A=@n28, B=@n27))
  ) {materialFunction=&HueShift}
  TextureSample(UVs=Add(A=TextureCoordinate(), B=@n12."Random Number  - needs custom uvs")) {samplerType=SAMPLERTYPE_LinearColor, texture=&HighContrastNoise}
 ]
}
