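# Unlit translucent material for animated score numbers, blending a digit mask with a plus/minus texture.
# ScoreNumberWPO handles scale and offset and outputs ripple timeline and other values intended for custom UVs.
# Stepped, screen-aligned noise jitters the glyph UVs for a glitch effect, with alpha and depth fade controlling opacity.
# Color layers an outline, rippling grid and world-aligned grid projected from behind the translucency, then adds noise and a gradient hue shift.
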Material /Game/UI/DamageText/Score/ScoreNumbers_mat
graph {
 properties {
  bEnableResponsiveAA=true
  blendMode=BLEND_Translucent
  numCustomizedUVs=8
  pixelDepthOffsetMode=PDOM_Legacy
  refractionMethod=RM_IndexOfRefraction
  shadingModel=MSM_Unlit
  usedShadingModels=Unlit
 }
 nodes [
  n1 = TextureSample!(UVs=TextureCoordinate() {uTiling=5, vTiling=5}, Tex, Bias=Constant() {r=1}, Apply View MipBias) {mipValueMode=TMVM_MipBias, samplerType=SAMPLERTYPE_LinearColor, texture=/Game/Gadgets/Assets/AirStrike/Textures/T_Airstrike_Grid_01}
  n2 = SphereMask(A=TextureCoordinate(), B=Constant() {r=0.5}) {attenuationRadius=0.5, hardnessPercent=0}
  n3 = OneMinus(None=@n2)
  n4 = MaterialFunctionCall(Float2=TextureCoordinate() {coordinateIndex=4}) {
   desc="R=Timeline for ripple G= Float Uniform"
   materialFunction=&BreakOutFloat2Components
  }
  n5 = Multiply(
   A=@n2
   B=TextureSample!(UVs=Add(A=@n3, B=@n4.R), Tex, Apply View MipBias) {samplerType=SAMPLERTYPE_LinearColor, texture=&/ScoreNumber_ripple}.RGB
  )
  n6 = ComponentMask(None=TextureCoordinate() {coordinateIndex=2}) {r=true}
  n7 = Multiply(A=Add(A=Multiply(A=@n1.R, B=@n5), B=Multiply(A=@n5) {constB=0.05}), B=OneMinus(None=@n6)) {constA=1}
  Add(A=@n7)
  n8 = MaterialFunctionCall!(
   Scale Modulation=ComponentMask(None=VectorParameter() {group=Visuals, parameterName=XYScaleModulation, defaultValue=<rgba 1 0.8 0 1>}.RGB) {g=true, r=true}
   Number Scale=ScalarParameter() {defaultValue=2, parameterName=NumberScale}
   "Number Offset Scale "=ScalarParameter() {defaultValue=0.9, parameterName=NumberOffsetScale}
   Debug texture coordinate input
   Animation
   "Additional Camera Offset "
   "Additonal Screenspace XY offset "
   Animation Lifespan Override
  ) {materialFunction=&/ScoreNumberWPO}
  ComponentMask(None=@n8."To Custom UVS - R- Timeline for rippple - G - Float Uniform - B - White ") {g=true, r=true}
  ComponentMask(None=@n8."To Custom UVS - R- Timeline for rippple - G - Float Uniform - B - White ") {b=true}
  Constant()
  n9 = MaterialFunctionCall(Float2=TextureCoordinate() {coordinateIndex=3, desc="x= B Randomization  y= A Alpha"}) {materialFunction=&BreakOutFloat2Components}
  n10 = Constant2Vector() {g=0.1, r=1}
  n11 = MaterialFunctionCall!(TextureResolution=Constant2Vector() {g=482.7549, r=100}) {materialFunction=/Engine/Functions/Engine_MaterialFunctions02/Texturing/ScreenAlignedPixelToPixelUVs}
  n12 = Constant2Vector() {g=1}
  n13 = Constant() {r=0.5}
  n14 = Multiply(A=@n8."Random Number  - needs custom uvs") {constB=33.3033}
  n15 = MaterialFunctionCall(XY=Add(
    A=Multiply(A=@n11.Result, B=@n12)
    B=MaterialFunctionCall!(TimeSpeed=@n13, Jump Amount Per Step, Number of Steps Per Unit of Time, Time Offset=@n14) {materialFunction=/Engine/Functions/Engine_MaterialFunctions02/SteppingPannerTime}.Result
   )) {materialFunction=&Swizzle}
  n16 = SquareRoot(None=TextureSample(UVs=Add(A=@n15.YX, B=@n14)) {texture=/Game/Effects/Fort_Effects/Textures/Tile/MotionBlur}.R)
  n17 = ScalarParameter() {defaultValue=1, parameterName=Opacity Modulation}
  n18 = Power(Base=ConstantBiasScale(None=TextureCoordinate()) {bias=-0.5, scale=2}) {constExponent=8}
  n19 = Multiply(
   A=Clamp(None=Multiply(A=@n9.G, B=@n17))
   B=MaterialFunctionCall!(f2=OneMinus(None=@n18), f3, f4) {materialFunction=/Game/Packages/Fortnite_UniqueMaterials_1/VectorComponentOps}.f2 Mul
  )
  n20 = Multiply(A=Multiply(A=Multiply(A=@n9.R, B=@n10), B=ConstantBiasScale(None=@n16) {bias=-0.5, scale=2}), B=@n19)
  n21 = Add(A=@n20, B=TextureCoordinate() {coordinateIndex=1})
  n22 = Add(A=@n20, B=TextureCoordinate() {vTiling=0.5}) {constB=0}
  n23 = MaterialFunctionCall(Float3=LinearInterpolate(
    A=TextureSampleParameter2D(UVs=@n21) {group=Visuals, parameterName=Diffuse, samplerType=SAMPLERTYPE_Masks, texture=&/Score_Numbers_mask}.RGB
    B=TextureSampleParameter2D(UVs=@n22) {parameterName=+- Texture, texture=/Game/UI/DamageText/Score/+-ScoreDiffuse.+-ScoreDiffuse}.RGB
    Alpha=@n6
   )) {materialFunction=&BreakOutFloat3Components}
  DepthFade(Opacity=Multiply(A=@n23.R, B=@n19), FadeDistance=Constant() {r=64})
  n24 = ScalarParameter() {defaultValue=-0.0408, parameterName=Color Gradient Shift On Top}
  n25 = ScalarParameter() {defaultValue=0.0183, parameterName=Color Gradient Shift On Bottom}
  n26 = MaterialFunctionCall(Float2=TextureCoordinate()) {materialFunction=&BreakOutFloat2Components}
  n27 = Multiply(
   A=ScalarParameter() {defaultValue=-0.0069, parameterName=Noise}
   B=MaterialFunctionCall!(Center Point World Position) {materialFunction=/Game/Packages/Fortnite_UniqueMaterials_1/Effects/WorldAlignedScreenCoordinates}.Result
  ) {constA=42811.2383}
  n28 = Add(
   A=MaterialFunctionCall(TimeSpeed=Constant() {r=0.35}) {materialFunction=&SteppingPannerTime}.Result
   B=Multiply(A=@n27) {constB=0.5}
  )
  n29 = Add(
   A=@n27
   B=MaterialFunctionCall(TimeSpeed=Constant() {r=0.5}) {materialFunction=&SteppingPannerTime}.Result
  )
  n30 = LinearInterpolate(B=Add(
    A=TextureSample(UVs=@n28) {samplerType=SAMPLERTYPE_Normal, texture=/Game/Global/Textures/Default/Noise/64x64Noise}.RGB
    B=TextureSample(UVs=@n29) {samplerType=SAMPLERTYPE_Normal, texture=&64x64Noise}.RGB
   )) {constA=0.5, constAlpha=0.025, constB=0.5}
  n31 = VectorParameter() {parameterName=Color12, defaultValue=<rgba 0.0268 0.0477 0.132 1>}
  n32 = Constant() {r=1}
  n33 = ScalarParameter() {defaultValue=0.8181, parameterName=Outilne Hardness}
  n34 = Power(Base=ComponentMask(None=SphereMask(A=@n23.G, B=@n32, Radius=@n33) {hardnessPercent=0}) {r=true}) {constExponent=4}
  n35 = SceneTexture() {sceneTextureId=PPI_WorldNormal}
  n36 = Constant3Vector() {constant=<rgba 0 0 1 0>}
  n37 = MaterialFunctionCall!(
   Diffuse Texture Object=TextureObject() {samplerType=SAMPLERTYPE_LinearColor, texture=&T_Airstrike_Grid_01}
   Spec Texture Object
   TextureSize=ScalarParameter() {defaultValue=96.2244, parameterName=World grid scale}
   --------------------------
   Use seperate settings for Z projection
   Use seperate texture for Z projection
   Texture Size Z projection
   Z Plane Diffuse Texture Object
   Z Plane Spec Texture Object
   "--------------------------      "
   ProjectionTransitionContrast
   Use Texture Alpha Heightmap to Refine the Transistion
   HeightLerpTransitionContrast
   PickTheDiffuseHeightChannel
   TheDiffuseHeightChannel
   "--------------------------       "
   NormalTextureObject
   World Space Normals
   "--------------------------  "
   Export Float 4 Diffuse
   Export Float 4 Spec
   "--------------------------     "
   WorldPosition=MaterialFunctionCall() {materialFunction=&WorldPositionBehindTranslucency}.Result
   World Space Normal=ShadingPathSwitch(Default=@n35.Color, Deferred=@n35.Color, Forward=@n36, Mobile=@n36)
  ) {materialFunction=/Engine/Functions/Engine_MaterialFunctions02/Texturing/WorldAlignedTextures_Complex}
  n38 = WorldPosition()
  n39 = MaterialFunctionCall() {materialFunction=&WorldPositionBehindTranslucency}
  n40 = ScalarParameter() {defaultValue=220.9397, parameterName=World Grid Falloff Radius}
  n41 = Multiply(
   A=Multiply(
    A=MaterialFunctionCall(Float3=@n37.XYZ Diffuse Texture) {materialFunction=&BreakOutFloat3Components}.R
    B=SphereMask(A=@n38.XYZ, B=@n39.Result, Radius=@n40) {hardnessPercent=0}
   )
   B=VectorParameter() {parameterName=World lines, defaultValue=<rgba 0.191 0.191 0.191 1>}.RGB
  )
  n42 = TextureCoordinate() {coordinateIndex=5}
  n43 = Add(
   A=Add(
    A=ComponentMask(None=@n42) {desc="R = Blue Dynamic White", r=true}
    B=VectorParameter() {parameterName=Color0, defaultValue=<rgba 0 0.0362 0.24 1>}.RGB
   )
   B=OneMinus(None=@n23.R)
  ) {constB=0}
  n44 = VectorParameter() {parameterName=Color1, defaultValue=<rgba 0.3513 0.5407 0.865 0.01>}
  n45 = Add(A=Add(A=@n23.R, B=Multiply(A=@n23.B, B=@n44.A)), B=Multiply(A=@n7) {constB=0.25})
  n46 = Add(A=Multiply(A=Multiply(A=@n31.RGB, B=@n34), B=@n23.R), B=Multiply(A=Add(A=@n41, B=@n43), B=Multiply(A=@n44.RGB, B=@n45)))
  n47 = Multiply(A=@n4.G, B=@n46)
  MaterialFunctionCall(
   Hue Shift Percentage=Add(A=Multiply(A=@n9.R) {constB=0.15}, B=LinearInterpolate(A=@n24, B=@n25, Alpha=@n26.G) {constA=1, constB=0.5})
   Texture=Multiply(A=Add(A=@n30) {constB=0.5}, B=Add(A=@n47, B=@n46))
  ) {materialFunction=&HueShift}
  ScalarParameter() {defaultValue=0.35, parameterName=DEBUG ANIMATION}
  TextureSample(UVs=Add(A=TextureCoordinate(), B=@n8."Random Number  - needs custom uvs")) {samplerType=SAMPLERTYPE_LinearColor, texture=&HighContrastNoise}
 ]
}
